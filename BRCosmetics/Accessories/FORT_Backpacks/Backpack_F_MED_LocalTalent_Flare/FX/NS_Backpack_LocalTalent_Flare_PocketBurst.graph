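# Music-reactive Niagara burst effect for the LocalTalent Flare backpack, with most emitters firing from two pocket offsets.
# Looping mesh flares, stars, waves, rays, music lines and sprite particles take spawn chance, brightness or scale from event audio amplitude and an interpolated moving average.
# System.MusicReactivity blends audio-driven values toward steady ones.
# User.VariantSwitch selects palettes, while bIsFrontEndPreview provides separate lobby preview behavior.
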NiagaraSystem /BRCosmetics/Accessories/FORT_Backpacks/Backpack_F_MED_LocalTalent_Flare/FX/NS_Backpack_LocalTalent_Flare_PocketBurst
dependencies {
 usedDataInterfaces [
  &NiagaraDataInterfaceArrayFloat3
  &NiagaraDataInterfaceColorCurve
  &NiagaraDataInterfaceCurve
  &NiagaraDataInterfaceArrayColor
  &NiagaraDataInterfacePlatformSet
  &NiagaraDataInterfaceArrayFloat
 ]
 usedDynamicInputs []
 usedModules [
  &self:SystemSpawnScript.SystemScriptSource.SystemScriptGraph.NiagaraNodeAssignment_0.SetVariables_#1
  &SystemState
  &InterpolateOverTime
 ]
}
emitters {
 Flare {
  emitterUpdateScript [
   &EmitterState(Inactive Response=::NewEnumerator1, Life Cycle Mode=::NewEnumerator1, Loop Duration=1, UseLoopDelay=false)
   &SpawnRate(
    Spawn Probability=&MakeCustomFloatFromBool(
     Bool=$User.bIsFrontEndPreview
     True Float=1
     False Float=&Lerp_Float(
      A=&MakeCustomFloatFromBool(
       Bool=&SetBoolByFloatComparison(A=$NPC.FortniteNPC.EventAudio_Amplitude, B=0, Comparison Type=::NewEnumerator4)
       True Float=0
       False Float=1
      )
      B=1
      Alpha=$System.MusicReactivity
     )
    )
    SpawnRate=6
    Use Spawn Probability=true
   )
  ]
  particleSpawnScript [
   &PartitionParticles()
   &InitializeParticle(
    Color=<rgba 0.2 0.2 0.2 20>
    Lifetime=0.6
    Mesh Scale=<0.08 0.16 0.16>
    Mesh Scale Mode=::NewEnumerator4
    Position Offset=&SelectVectorFromArray(
     Direct Array Index=$Particles.PartitionParticles.Partition
     Vector Selection Array=DataInterface() {propertyValues {FloatData [<7 -5 10>, <-8 -5 -11>]}}
    )
    UsePositionOffset=true
    Write Material Random=true
   ) {module=InitializeParticle001}
   &InitialMeshOrientation(Mesh Orientation Mode=::NewEnumerator2, Rotation=<0 0 -0.25>, Use Rotation Vector=true)
  ]
  particleUpdateScript [
   &ParticleState()
   &ScaleColor(
    Scale Alpha=&FloatFromCurve(
     FloatCurve=DataInterface() {propertyValues {Curve {keys [
         {arriveTangent=-3.8906, interpMode=RCIM_Cubic, leaveTangent=-3.8906, tangentMode=RCTM_User, time=0, value=1}
         {arriveTangent=-1.0522, interpMode=RCIM_Cubic, leaveTangent=-1.0522, tangentMode=RCTM_User, time=0.2, value=0.4}
         <at 1 0 RCIM_Cubic>
        ]}}}
     Scale Curve=&Multiply_Float(
      A=&FN_FrontEndFloatModifier(Gameplay=1, Front End=0.5)
      B=&Multiply_Float(
       A=&FNTimeOfDayScalar(Night Scale=0.5, Morning Scale=1, Evening Scale=1)
       B=&FN_PickFloatDependOnPlatform(TargetPlatformActive=1, TargetPlatformNoActive=0.5)
      )
     )
    )
    Scale RGB=&Multiply_Vector(
     A=/Niagara/DynamicInputs/Math/LerpMultipleVectors(
      Alpha:NiagaraFloat=$User.VariantSwitch
      V 1:Vector3f=&MakeVectorFromLinearColorRGB(Color=&ColorFromCurve(ColorCurve=DataInterface() {propertyValues {
          AlphaCurve {keys []}
          BlueCurve {keys [<at 0.1142 1 RCIM_Cubic>,<at 0.259 0.6136>,<at 0.401 1>]}
          GreenCurve {keys [<at 0.1142 0.4512 RCIM_Cubic>,<at 0.259 2>,<at 0.401 0.6265>]}
          RedCurve {keys [<at 0.1142 0 RCIM_Cubic>,<at 0.259 0.0013>,<at 0.401 0>]}
         }}))
      V 2:Vector3f=&MakeVectorFromLinearColorRGB(Color=&ColorFromCurve(ColorCurve=DataInterface() {propertyValues {
          AlphaCurve {keys []}
          BlueCurve {keys [<at 0.118 0 RCIM_Cubic>,<at 0.2565 0>,<at 0.4055 0 RCIM_Cubic>]}
          GreenCurve {keys [<at 0.118 0.9271 RCIM_Cubic>,<at 0.2565 2>,<at 0.4055 0.9271 RCIM_Cubic>]}
          RedCurve {keys [<at 0.118 0.181 RCIM_Cubic>,<at 0.2565 0.4026>,<at 0.4055 0.181 RCIM_Cubic>]}
         }}))
      V 3:Vector3f=?
      V 4:Vector3f=?
      V 5:Vector3f=?
      Vector Count:ENiagara_MultipleLerpCount=::NewEnumerator3
     )
     B=&VectorFromFloat(Value=&FloatFromCurve(
       FloatCurve=DataInterface() {propertyValues {Curve {keys [{arriveTangent=-1.3603, interpMode=RCIM_Cubic, leaveTangent=-1.3603, tangentMode=RCTM_User, time=0, value=0.8}, <at 1 0 RCIM_Cubic>]}}}
       Scale Curve=0.3
      ))
    )
    ScaleA=true
    ScaleRGB=true
   ) {module=ScaleColor001}
   &ScaleColor(
    Scale Alpha=&Lerp_Float(
     A=&FloatFromCurve(
      FloatCurve=DataInterface() {propertyValues {Curve {keys [
          {interpMode=RCIM_Cubic, tangentMode=RCTM_User, time=0.1, value=0}
          {arriveTangent=7.2184, interpMode=RCIM_Cubic, tangentMode=RCTM_Break, time=0.4, value=1}
         ]}}}
      CurveIndex=$System.InterpolateOverTime.InterpolateOverTime_Float.Moving Average
     )
     B=1
     Alpha=$System.MusicReactivity
    )
    ScaleA=true
    ScaleRGB=true
   )
   &ScaleMeshSize(Scale Factor=&VectorFromFloat(Value=&FloatFromCurve(FloatCurve=DataInterface() {propertyValues {Curve {keys [<at 0 0>,<at 0.2 1.375>,<at 1 2>]}}}, Scale Curve=0.85)))
   &DynamicMaterialParameters(
    Index 0 Param 1=&FloatFromCurve(FloatCurve=DataInterface() {propertyValues {Curve {keys [{arriveTangent=-0.466, interpMode=RCIM_Cubic, leaveTangent=-0.466, tangentMode=RCTM_User, time=0, value=0.5}, <at 1 0.2>]}}})
    Index 0 Param 2=&FloatFromCurve(FloatCurve=DataInterface() {propertyValues {Curve {keys [{arriveTangent=-0.1866, interpMode=RCIM_Cubic, leaveTangent=-0.1866, tangentMode=RCTM_User, time=0, value=1}, <at 1 0.9>]}}})
    Index 0 Param 3=&FloatFromCurve(FloatCurve=DataInterface() {propertyValues {Curve {keys [<at 0 0>,<at 0.15 1>,<at 1 1.2>]}}})
    Module.Param0WriteEnabled=true
    Module.Param1WriteEnabled=true
    Module.Param2WriteEnabled=true
    Module.Param3WriteEnabled=false
   )
   /SetVariables(Particles.Dyn002:Vector4f=&MakeVector4(
     X=&FloatFromCurve(
      FloatCurve=DataInterface() {propertyValues {Curve {keys [{arriveTangent=-0.6916, interpMode=RCIM_Cubic, leaveTangent=-0.6916, tangentMode=RCTM_User, time=0, value=0.5}, <at 1 0.1>]}}}
      Scale Curve=0.6
     )
     Y=&FloatFromCurve(
      FloatCurve=DataInterface() {propertyValues {Curve {keys [{arriveTangent=-0.1866, interpMode=RCIM_Cubic, leaveTangent=-0.1866, tangentMode=RCTM_User, time=0, value=1}, <at 1 0.9>]}}}
      Scale Curve=0.4
     )
     Z=&FloatFromCurve(FloatCurve=DataInterface() {propertyValues {Curve {keys [<at 0 0>,<at 0.15 1>,<at 1 1.2>]}}})
     W=0
    )) {isSetParameters=true}
  ]
  rendererClasses [&NiagaraMeshRendererProperties]
  renderers [{rendererClass=&NiagaraMeshRendererProperties, rendererIndex=0}]
 }
 MusicLine {
  emitterUpdateScript [
   &EmitterState(Inactive Response=::NewEnumerator1, Life Cycle Mode=::NewEnumerator1, Loop Duration=0.5, UseLoopDelay=false)
   &SpawnBurst_Instantaneous(
    Spawn Probability=&MakeCustomFloatFromBool(
     Bool=$User.bIsFrontEndPreview
     True Float=1
     False Float=&Lerp_Float(
      A=&MakeCustomFloatFromBool(
       Bool=&SetBoolByFloatComparison(A=$NPC.FortniteNPC.EventAudio_Amplitude, B=0, Comparison Type=::NewEnumerator4)
       True Float=0
       False Float=1
      )
      B=1
      Alpha=$System.MusicReactivity
     )
    )
    Use Spawn Probability=true
    UseLoopCountLimit=false
   )
  ]
  particleSpawnScript [
   &InitializeParticle(
    Lifetime=0.55
    Mesh Scale=&Multiply_VectorByFloat(Vector=<1 0.1 0.45>, Float=1)
    Mesh Scale Mode=::NewEnumerator4
    Position Offset=<-10 -3 6>
    UsePositionOffset=true
    Write Material Random=true
   )
   &InitialMeshOrientation(Mesh Orientation Mode=::NewEnumerator2, Rotation=<0 -0.15 0>, Use Rotation Vector=true)
  ]
  particleUpdateScript [
   &ParticleState()
   &DynamicMaterialParameters(
    Index 0 Param 1=&Lerp_Float(
     A=&Multiply_Float(
      A=&Lerp_Float(A=0.5, B=1.5, Alpha=$System.InterpolateOverTime.InterpolateOverTime_Float.Moving Average)
      B=&ClampFloat(Float=$NPC.FortniteNPC.EventAudio_Amplitude)
     )
     B=1
     Alpha=$System.MusicReactivity
    )
    Index 0 Param 2=&Lerp_Float(
     A=&Multiply_Float(
      A=&Lerp_Float(A=2, B=5, Alpha=$System.InterpolateOverTime.InterpolateOverTime_Float.Moving Average)
      B=&ClampFloat(Float=$NPC.FortniteNPC.EventAudio_Amplitude)
     )
     B=3
     Alpha=$System.MusicReactivity
    )
    Index 1 Param 1=2
    Index 1 Param 2=2
    Module.Param0WriteEnabled=true
    Module.Param0WriteEnabled1=true
    Module.Param1WriteEnabled=true
    Module.Param1WriteEnabled1=true
    Module.Param2WriteEnabled=false
    Module.Param2WriteEnabled1=false
    Module.Param3WriteEnabled=false
    Module.Param3WriteEnabled1=false
    Write Parameter Index 1=true
   ) {module=DynamicMaterialParameters001}
   &ScaleMeshSize(Scale Factor=&MakeVector(
     X=1
     Y=1
     Z=&FloatFromCurve(FloatCurve=DataInterface() {propertyValues {Curve {keys [<at 0 0>,<at 0.75 1>,<at 1 2>]}}})
    )) {enabled=false}
   /SetVariables(Particles.Size003:Vector3f=<1 0.1 0.6>) {isSetParameters=true}
   /SetVariables(Particles.Scale_Beam:Vector3f=<0.8 7 0.3>) {isSetParameters=true}
   /SetVariables(Particles.Scale_OutsideGlow:Vector3f=<1.5 1 1.2>) {isSetParameters=true}
   &ScaleColor(
    Scale Alpha=&Multiply_Float(
     A=0.22
     B=&Lerp_Float(
      A=&Multiply_Float(
       A=&Lerp_Float(A=0, B=1, Alpha=$System.InterpolateOverTime.InterpolateOverTime_Float.Moving Average)
       B=&ClampFloat(Float=$NPC.FortniteNPC.EventAudio_Amplitude)
      )
      B=1
      Alpha=$System.MusicReactivity
     )
    )
    Scale RGB=&MakeVectorFromLinearColorRGB(Color=&SelectLinearColorFromArray(
      Array Sampling Mode=::NewEnumerator1
      Direct Array Index=&MakeIntFromFloat(Float=$User.VariantSwitch)
      Color Selection Array=DataInterface() {propertyValues {ColorData [<rgba 0 1 0.6671 1>, <rgba 0.5949 0.9 0.0828 1>]}}
      Random Seed=?
     ))
    ScaleA=true
    ScaleRGB=true
   )
   /SetVariables(Particles.Color002=&MakeLinearColorFromVectorAndFloat(
     "Vector (RGB)"=&MakeVectorFromLinearColorRGB(Color=&SelectLinearColorFromArray(
       Array Sampling Mode=::NewEnumerator1
       Direct Array Index=&MakeIntFromFloat(Float=$User.VariantSwitch)
       Color Selection Array=DataInterface() {propertyValues {ColorData [<rgba 0 0.5 0.4127 1>, <rgba 0.0717 0.2812 0.0207 1>]}}
       Random Seed=?
      ))
     "Float (Alpha)"=&Multiply_Float(
      A=1
      B=&Lerp_Float(
       A=&Multiply_Float(
        A=&Lerp_Float(A=0, B=1, Alpha=$System.InterpolateOverTime.InterpolateOverTime_Float.Moving Average)
        B=&ClampFloat(Float=$NPC.FortniteNPC.EventAudio_Amplitude)
       )
       B=1
       Alpha=$System.MusicReactivity
      )
     )
    )) {isSetParameters=true}
   /SetVariables(Particles.Color003:LinearColor=&MakeLinearColorFromVectorAndFloat(
     "Vector (RGB)"=&MakeVectorFromLinearColorRGB(Color=&SelectLinearColorFromArray(
       Array Sampling Mode=::NewEnumerator1
       Direct Array Index=&MakeIntFromFloat(Float=$User.VariantSwitch)
       Color Selection Array=DataInterface() {propertyValues {ColorData [<rgba 0.3945 1 0 1>, <rgba 0.092 0.827 1 1>]}}
       Random Seed=?
      ))
     "Float (Alpha)"=&Lerp_Float(
      A=&Multiply_Float(
       A=&Lerp_Float(A=0, B=2, Alpha=$System.InterpolateOverTime.InterpolateOverTime_Float.Moving Average)
       B=&ClampFloat(Float=$NPC.FortniteNPC.EventAudio_Amplitude)
      )
      B=1
      Alpha=$System.MusicReactivity
     )
    )) {isSetParameters=true}
   /SetVariables(Particles.Color_OutsideGlow:LinearColor=&MakeLinearColorFromVectorAndFloat(
     "Vector (RGB)"=&MakeVectorFromLinearColorRGB(Color=&SelectLinearColorFromArray(
       Array Sampling Mode=::NewEnumerator1
       Direct Array Index=&MakeIntFromFloat(Float=$User.VariantSwitch)
       Color Selection Array=DataInterface() {propertyValues {ColorData [<rgba 0.7636 1 0.1026 1>, <rgba 0 0.4061 1 1>]}}
       Random Seed=?
      ))
     "Float (Alpha)"=&FloatFromCurve(
      FloatCurve=DataInterface() {propertyValues {Curve {keys [<at 0 0>,<at 0.1 1>,<at 0.9 1>,<at 1 0>]}}}
      Scale Curve=&Lerp_Float(A=0, B=0.3, Alpha=$System.InterpolateOverTime.InterpolateOverTime_Float.Moving Average)
     )
    )) {isSetParameters=true}
  ]
  rendererClasses [&NiagaraMeshRendererProperties]
  renderers [
   {rendererClass=&NiagaraMeshRendererProperties, rendererIndex=0}
   {rendererClass=&NiagaraMeshRendererProperties, rendererIndex=1}
   {rendererClass=&NiagaraMeshRendererProperties, rendererIndex=2}
   {rendererClass=&NiagaraMeshRendererProperties, rendererIndex=3}
   {rendererClass=&NiagaraMeshRendererProperties, rendererIndex=4}
  ]
 }
 Rays {
  emitterUpdateScript [
   &EmitterState(Inactive Response=::NewEnumerator1, Life Cycle Mode=::NewEnumerator1, Loop Duration=0.5, UseLoopDelay=false)
   &SpawnBurst_Instantaneous(
    Spawn Count=2
    Spawn Probability=&MakeCustomFloatFromBool(
     Bool=$User.bIsFrontEndPreview
     True Float=1
     False Float=&Lerp_Float(
      A=&MakeCustomFloatFromBool(
       Bool=&SetBoolByFloatComparison(A=$NPC.FortniteNPC.EventAudio_Amplitude, B=0, Comparison Type=::NewEnumerator4)
       True Float=0
       False Float=1
      )
      B=1
      Alpha=$System.MusicReactivity
     )
    )
    Use Spawn Probability=true
    UseLoopCountLimit=false
   )
  ]
  particleSpawnScript [
   &PartitionParticles()
   &InitializeParticle(
    Lifetime=0.5
    Mesh Scale=<0.3 -1 -1>
    Mesh Scale Mode=::NewEnumerator4
    Position Offset=&SelectVectorFromArray(
     Direct Array Index=$Particles.PartitionParticles.Partition
     Vector Selection Array=DataInterface() {propertyValues {FloatData [<7.2 -6 9.8>, <-7.5 -6 -11.1>]}}
    )
    Sprite Rotation Angle=&RandomRangeFloat(Minimum=0, Maximum=1)
    Sprite Rotation Mode=::NewEnumerator1
    UsePositionOffset=true
    Write Material Random=true
   )
   &DynamicMaterialParameters(
    Index 0 Param 1=&RandomRangeFloat(Minimum=0, Maximum=10)
    Index 0 Param 2=&RandomRangeFloat(Minimum=2.5, Maximum=4.5)
    Index 0 Param 3=&RandomRangeFloat(Minimum=0.8, Maximum=1.6)
    Module.Param0WriteEnabled=true
    Module.Param1WriteEnabled=true
    Module.Param2WriteEnabled=true
    Module.Param3WriteEnabled=false
   )
  ]
  particleUpdateScript [
   &ParticleState()
   &UpdateMeshOrientation(Rotation Rate=0.7, Rotation Vector=<0 -1 0>)
   &ScaleMeshSize(Scale Factor=<0.4 0.4 0.4>) {module=ScaleMeshSize001}
   &DynamicMaterialParameters(
    Index 0 Param 4=1
    Module.Param0WriteEnabled=false
    Module.Param1WriteEnabled=false
    Module.Param2WriteEnabled=false
    Module.Param3WriteEnabled=true
   ) {module=DynamicMaterialParameters001}
   /SetVariables(Particles.scale_small:Vector3f=&VectorFromFloat(Value=0.71)) {isSetParameters=true}
   /SetVariables(Particles.Color002=&MakeLinearColorFromVectorAndFloat(
     "Vector (RGB)"=&Multiply_VectorByFloat(
      Vector=&LerpMultipleVectors(
       Alpha:NiagaraFloat=$User.VariantSwitch
       V 1:Vector3f=&MakeVectorFromLinearColorRGB(Color=&ColorFromCurve(ColorCurve=DataInterface() {propertyValues {
           AlphaCurve {keys []}
           BlueCurve {keys [<at 0.0776 0>,<at 0.2618 0 RCIM_Cubic>,<at 0.6449 0>,<at 0.9777 0>]}
           GreenCurve {keys [<at 0.0776 0.9271>, {arriveTangent=0.1285, interpMode=RCIM_Cubic, leaveTangent=0.1285, time=0.2618, value=1}, <at 0.6449 1>, <at 0.9777 1>]}
           RedCurve {keys [
             <at 0.0776 0.181>
             {arriveTangent=-0.2504, interpMode=RCIM_Cubic, leaveTangent=-0.2504, time=0.2618, value=0.0389}
             <at 0.6449 0.0389>
             <at 0.9777 0.3945>
            ]}
          }}))
       V 2:Vector3f=&MakeVectorFromLinearColorRGB(Color=&ColorFromCurve(ColorCurve=DataInterface() {propertyValues {
           AlphaCurve {keys []}
           BlueCurve {keys [
             <at 0.0815 0 RCIM_Cubic>
             {arriveTangent=0.0643, interpMode=RCIM_Cubic, leaveTangent=0.0643, time=0.2671, value=0.0306}
             <at 0.5574 0.0306>
             <at 0.8385 0>
            ]}
           GreenCurve {keys [
             <at 0.0815 0.8329 RCIM_Cubic>
             {arriveTangent=-0.0691, interpMode=RCIM_Cubic, leaveTangent=-0.0691, time=0.2671, value=0.8}
             <at 0.5574 0.8>
             <at 0.8385 0.8329>
            ]}
           RedCurve {keys [
             <at 0.0815 0.9271 RCIM_Cubic>
             {arriveTangent=-1.7553, interpMode=RCIM_Cubic, leaveTangent=-1.7553, time=0.2671, value=0.0917}
             <at 0.5574 0.0917>
             <at 0.8385 0.9271>
            ]}
          }}))
       V 3:Vector3f=?
       V 4:Vector3f=?
       V 5:Vector3f=?
       Vector Count:ENiagara_MultipleLerpCount=::NewEnumerator3
      )
      Float=1
     )
     "Float (Alpha)"=&Multiply_Float(
      A=0.8
      B=&MakeCustomFloatFromBool(
       Bool=$User.bIsFrontEndPreview
       True Float=&Lerp_Float(A=0.5, B=1, Alpha=$System.InterpolateOverTime.InterpolateOverTime_Float.Moving Average)
       False Float=1
      )
     )
    )) {isSetParameters=true}
   /SetVariables() {isSetParameters=true}
  ]
  rendererClasses [&NiagaraMeshRendererProperties]
  renderers [{rendererClass=&NiagaraMeshRendererProperties, rendererIndex=0}]
 }
 Sprites {
  emitterUpdateScript [
   &EmitterState(
    Enable Distance Culling=true
    Inactive Response=::NewEnumerator1
    Life Cycle Mode=::NewEnumerator1
    Loop Duration=1
    Max Distance Response=::KillImmediately
    Module.bUseMaxDistance=true
    Module.bUseMinDistance=false
    Scalability Mode=::NewEnumerator1
    UseLoopDelay=false
   )
   &SpawnRate(
    Spawn Probability=&MakeCustomFloatFromBool(
     Bool=$User.bIsFrontEndPreview
     True Float=&Lerp_Float(A=0.16, B=1, Alpha=$System.InterpolateOverTime.InterpolateOverTime_Float.Moving Average)
     False Float=&Lerp_Float(
      A=&MakeCustomFloatFromBool(
       Bool=&SetBoolByFloatComparison(A=$NPC.FortniteNPC.EventAudio_Amplitude, B=0, Comparison Type=::NewEnumerator4)
       True Float=0
       False Float=&FloatFromCurve(
        FloatCurve=DataInterface() {propertyValues {Curve {keys [
            {interpMode=RCIM_Cubic, tangentMode=RCTM_User, time=0.1, value=0}
            {arriveTangent=7.2184, interpMode=RCIM_Cubic, tangentMode=RCTM_Break, time=0.4, value=1}
           ]}}}
        CurveIndex=$System.InterpolateOverTime.InterpolateOverTime_Float.Moving Average
       )
      )
      B=1
      Alpha=$System.MusicReactivity
     )
    )
    SpawnRate=50
    Use Spawn Probability=true
   )
  ]
  particleSpawnScript [
   &InitializeParticle(
    Color=&MakeCustomLinearColorFromBool(
     Bool=&RandomBool()
     True Color=<rgba 0 0.7188 0.2139 1>
     False Color=&MakeCustomLinearColorFromBool(Bool=&RandomBool(), True Color=<rgba 0.7658 1 0 1>, False Color=<rgba 0 0.6164 0.6979 1>)
    )
    Lifetime Max=0.8
    Lifetime Min=0.5
    Lifetime Mode=::NewEnumerator1
    Position Offset=<0 -3 0>
    Sprite Rotation Mode=::NewEnumerator2
    Sprite Size Mode=::NewEnumerator1
    Uniform Sprite Size Min=6
    UsePositionOffset=true
    Write Material Random=true
   )
   &ShapeLocation(
    Cylinder Height=40
    Cylinder Radius=5
    MaskSpawnGroup=false
    Offset=<0 -5 0>
    Offset Mode=::NewEnumerator0
    Override Seed=false
    Rotation Axis=<0.1 0 0.15>
    Rotation Mode=::NewEnumerator1
    Shape Primitive=::NewEnumerator1
    Surface Only=true
    Surface Only Band Thickness=0.1
    Use Endcaps In Surface Only Mode=false
    UseOwnerScale=true
   )
   &AddVelocity(
    Module.Use Falloff Distance=false
    Origin Offset=<0 1 1>
    Override Seed=false
    Velocity Mode=::NewEnumerator1
    Velocity Speed=&RandomRangeFloat(Minimum=20, Maximum=30)
   )
   /SetVariables(Particles.VisibilityTag=&RandomRangeInt(Minimum=0, Maximum=3)) {isSetParameters=true}
   /SetVariables(Particles.Color_triangle:LinearColor=&MakeCustomLinearColorFromBool(
     Bool=&RandomBool()
     True Color=<rgba 0 0.7188 0.2139 1>
     False Color=&MakeCustomLinearColorFromBool(Bool=&RandomBool(), True Color=<rgba 0.4456 1 0 1>, False Color=<rgba 0 0.6348 0.7188 1>)
    )) {isSetParameters=true}
   /SetVariables(Particles.Color_triangleglow:LinearColor=&MultiplyLinearColor(
     Linear Color=$Particles.Color_triangle
     Scale Factor=&MakeVector4_FromLinearColor(LinearColor=/Niagara/DynamicInputs/TypeConversions/MakeLinearColorFromFloat(Value:NiagaraFloat=0.5))
    )) {isSetParameters=true}
  ]
  particleUpdateScript [
   &ParticleState()
   &SpriteRotationRate(Rotation Rate=&RandomRangeFloat(Minimum=-200, Maximum=200))
   &ScaleSpriteSize(Scale Factor=&Vector2DFromFloat(Value=&FloatFromCurve(
      FloatCurve=DataInterface() {propertyValues {Curve {keys [
          {interpMode=RCIM_Cubic, tangentMode=RCTM_User, time=0, value=0}
          {interpMode=RCIM_Cubic, tangentMode=RCTM_User, time=0.4, value=1}
          {tangentMode=RCTM_User, time=1, value=0}
         ]}}}
      Scale Curve=0.8
     )))
   /SetVariables(Particles.Size002:Vector2f=&Multiply_Vector2D(A=$Particles.SpriteSize, B=&Vector2DFromFloat(Value=1.5))) {isSetParameters=true}
   /SetVariables(Particles.Size_Triangle:Vector2f=&Multiply_Vector2D(A=$Particles.SpriteSize, B=&Vector2DFromFloat(Value=0.65))) {isSetParameters=true}
   &Drag(Drag=&RandomRangeFloat(Minimum=1, Maximum=2), Use Linear Drag=true, Use Rotational Drag=true)
   &AccelerationForce(Acceleration=&MakeCustomVectorFromBool(Bool=$User.bIsFrontEndPreview, True Vector=<0 0 20>, False Vector=<1 50 20>))
   &SolveForcesAndVelocity(Clamp Velocity=false, Limit Acceleration=false)
   &ScaleColor(
    Scale Alpha=&Multiply_Float(
     A=&FNTimeOfDayScalar(Night Scale=1, Morning Scale=0.3, Evening Scale=0.3)
     B=&FloatFromCurve(FloatCurve=DataInterface() {propertyValues {Curve {keys [
          {interpMode=RCIM_Cubic, tangentMode=RCTM_User, time=0.1, value=0}
          {interpMode=RCIM_Cubic, tangentMode=RCTM_User, time=0.3, value=0.8}
          {tangentMode=RCTM_User, time=0.8, value=0}
         ]}}})
    )
    Scale RGB=&MakeVectorFromLinearColorRGB(Color=&SelectLinearColorFromArray(
      Array Sampling Mode=::NewEnumerator1
      Direct Array Index=&MakeIntFromFloat(Float=$User.VariantSwitch)
      Color Selection Array=DataInterface() {propertyValues {ColorData [<rgba 1 1 1 1>, <rgba 0 0.6174 1 1>]}}
      Random Seed=?
     ))
    ScaleA=true
    ScaleRGB=true
   )
   /SetVariables(Particles.Color002=&MultiplyLinearColor(Linear Color=<rgba 1.2 1.2 1.2 1>, Scale Factor=&MakeVector4_FromLinearColor(LinearColor=$Particles.Color))) {isSetParameters=true}
  ]
  rendererClasses [&NiagaraSpriteRendererProperties]
  renderers [
   {rendererClass=&NiagaraSpriteRendererProperties}
   {rendererClass=&NiagaraSpriteRendererProperties, rendererIndex=1}
   {rendererClass=&NiagaraSpriteRendererProperties, rendererIndex=2}
   {rendererClass=&NiagaraSpriteRendererProperties, rendererIndex=3}
  ]
 }
 Star_Glowmesh {
  emitterUpdateScript [
   &EmitterState(Inactive Response=::NewEnumerator1, Life Cycle Mode=::NewEnumerator1, Loop Duration=1, UseLoopDelay=false)
   &SpawnRate(
    Spawn Probability=&MakeCustomFloatFromBool(
     Bool=$User.bIsFrontEndPreview
     True Float=1
     False Float=&Lerp_Float(
      A=&MakeCustomFloatFromBool(
       Bool=&SetBoolByFloatComparison(A=$NPC.FortniteNPC.EventAudio_Amplitude, B=0, Comparison Type=::NewEnumerator4)
       True Float=0
       False Float=1
      )
      B=1
      Alpha=$System.MusicReactivity
     )
    )
    SpawnRate=6
    Use Spawn Probability=true
   )
  ]
  particleSpawnScript [
   &PartitionParticles()
   &InitializeParticle(
    Lifetime=0.5
    Mesh Scale=<0.6 0.6 1.5>
    Mesh Scale Mode=::NewEnumerator4
    Position Offset=&SelectVectorFromArray(
     Direct Array Index=$Particles.PartitionParticles.Partition
     Vector Selection Array=DataInterface() {propertyValues {FloatData [<7 -5 10>, <-8 -5 -11>]}}
    )
    Sprite Size Mode=::NewEnumerator3
    Uniform Sprite Size=23
    UsePositionOffset=true
    Write Material Random=true
   )
   &InitialMeshOrientation(Mesh Orientation Mode=::NewEnumerator2, Rotation=<0.25 0 0>, Use Rotation Vector=true)
   &InitialMeshOrientation(
    Mesh Orientation Mode=::NewEnumerator2
    Rotation=&RandomRangeVector(Minimum=<0 0 0>, Maximum=<0 1 0>)
    Use Rotation Vector=true
   ) {module=InitialMeshOrientation001}
  ]
  particleUpdateScript [
   &ParticleState()
   &ScaleMeshSize(Scale Factor=&MakeVector(
     X=1
     Y=1
     Z=&FloatFromCurve(FloatCurve=DataInterface() {propertyValues {Curve {keys [
          {arriveTangent=1.9289, interpMode=RCIM_Cubic, leaveTangent=1.9289, tangentMode=RCTM_User, time=0, value=0.3}
          {arriveTangent=0.0001, interpMode=RCIM_Cubic, leaveTangent=0.0001, tangentMode=RCTM_User, time=1, value=1.1}
         ]}}})
    ))
   &ScaleColor(
    Scale Alpha=&FloatFromCurve(FloatCurve=DataInterface() {propertyValues {Curve {keys [
         <at 0 0 RCIM_Cubic>
         {arriveTangent=0.0083, interpMode=RCIM_Cubic, leaveTangent=0.0083, tangentMode=RCTM_User, time=0.3, value=1}
         <at 1 0 RCIM_Cubic>
        ]}}})
    Scale RGB=&Multiply_VectorByFloat(
     Vector=&LerpMultipleVectors(
      Alpha:NiagaraFloat=$User.VariantSwitch
      V 1:Vector3f=&MakeVectorFromLinearColorRGB(Color=&ColorFromCurve(ColorCurve=DataInterface() {propertyValues {
          AlphaCurve {keys []}
          BlueCurve {keys [{arriveTangent=1.6431, interpMode=RCIM_Cubic, time=0.3829, value=0}, <at 0.4534 1>]}
          GreenCurve {keys [{arriveTangent=-0.9364, interpMode=RCIM_Cubic, time=0.3829, value=1}, <at 0.4534 0.4312>]}
          RedCurve {keys [{arriveTangent=-1.6426, interpMode=RCIM_Cubic, time=0.3829, value=0.0389}, <at 0.4534 0>]}
         }}))
      V 2:Vector3f=&MakeVectorFromLinearColorRGB(Color=&ColorFromCurve(ColorCurve=DataInterface() {propertyValues {
          AlphaCurve {keys []}
          BlueCurve {keys [<at 0.3843 0.0635 RCIM_Cubic>,<at 0.4638 0.0676 RCIM_Cubic>]}
          GreenCurve {keys [<at 0.3843 0.7708 RCIM_Cubic>,<at 0.4638 0.9271 RCIM_Cubic>]}
          RedCurve {keys [<at 0.3843 0.6112 RCIM_Cubic>,<at 0.4638 0.1604 RCIM_Cubic>]}
         }}))
      V 3:Vector3f=?
      V 4:Vector3f=?
      V 5:Vector3f=?
      Vector Count:ENiagara_MultipleLerpCount=::NewEnumerator3
     )
     Float=1
    )
    ScaleA=true
    ScaleRGB=true
   )
   &ScaleColor(
    Scale Alpha=&Lerp_Float(
     A=&MakeCustomFloatFromBool(
      Bool=$User.bIsFrontEndPreview
      True Float=&Lerp_Float(A=0.12, B=1, Alpha=$System.InterpolateOverTime.InterpolateOverTime_Float.Moving Average)
      False Float=&FloatFromCurve(
       FloatCurve=DataInterface() {propertyValues {Curve {keys [
           {interpMode=RCIM_Cubic, tangentMode=RCTM_User, time=0.1, value=0}
           {arriveTangent=7.2184, interpMode=RCIM_Cubic, tangentMode=RCTM_Break, time=0.4, value=1}
          ]}}}
       CurveIndex=$System.InterpolateOverTime.InterpolateOverTime_Float.Moving Average
      )
     )
     B=1
     Alpha=$System.MusicReactivity
    )
    ScaleA=true
    ScaleRGB=true
   ) {module=ScaleColor001}
   &SolveForcesAndVelocity(Clamp Velocity=false, Limit Acceleration=false)
  ]
  rendererClasses [&NiagaraMeshRendererProperties]
  renderers [{rendererClass=&NiagaraMeshRendererProperties, rendererIndex=0}]
 }
 Star_mesh {
  emitterUpdateScript [
   &EmitterState(Inactive Response=::NewEnumerator1, Life Cycle Mode=::NewEnumerator1, Loop Duration=1, UseLoopDelay=false)
   &SpawnRate(
    Spawn Probability=&MakeCustomFloatFromBool(
     Bool=$User.bIsFrontEndPreview
     True Float=1
     False Float=&Lerp_Float(
      A=&MakeCustomFloatFromBool(
       Bool=&SetBoolByFloatComparison(A=$NPC.FortniteNPC.EventAudio_Amplitude, B=0, Comparison Type=::NewEnumerator4)
       True Float=0
       False Float=1
      )
      B=1
      Alpha=$System.MusicReactivity
     )
    )
    SpawnRate=6
    Use Spawn Probability=true
   )
  ]
  particleSpawnScript [
   &PartitionParticles()
   &InitializeParticle(
    Color Mode=::NewEnumerator0
    Lifetime=0.6
    Mesh Scale=<0.5 1 1>
    Mesh Scale Mode=::NewEnumerator4
    Position Offset=&SelectVectorFromArray(
     Direct Array Index=$Particles.PartitionParticles.Partition
     Vector Selection Array=DataInterface() {propertyValues {FloatData [<7 -6 10>, <-8 -6 -11>]}}
    )
    UsePositionOffset=true
    Write Material Random=true
   )
   &InitialMeshOrientation(Mesh Orientation Mode=::NewEnumerator2, Rotation=<0 0 -0.25>, Use Rotation Vector=true)
  ]
  particleUpdateScript [
   &ParticleState()
   &ScaleMeshSize(Scale Factor=&Multiply_Vector(
     A=<0.5 0.5 0.5>
     B=&VectorFromFloat(Value=&FloatFromCurve(
       FloatCurve=DataInterface() {propertyValues {Curve {keys [{arriveTangent=1.8728, interpMode=RCIM_Cubic, leaveTangent=1.8728, tangentMode=RCTM_User, time=0, value=1}, <at 1 2 RCIM_Cubic>]}}}
       Scale Curve=&MakeCustomFloatFromBool(Bool=$User.bIsFrontEndPreview, True Float=0.8, False Float=1)
      ))
    ))
   &ScaleColor(
    Scale Alpha=&Multiply_Float(
     A=&FloatFromCurve(
      FloatCurve=DataInterface() {propertyValues {Curve {keys [
          {arriveTangent=0.9513, interpMode=RCIM_Cubic, time=0, value=0}
          {arriveTangent=0.0508, interpMode=RCIM_Cubic, leaveTangent=0.0508, tangentMode=RCTM_User, time=0.2, value=1}
          {arriveTangent=-1.8779, interpMode=RCIM_Cubic, leaveTangent=-1.8779, tangentMode=RCTM_User, time=0.4, value=0.5}
          {interpMode=RCIM_Cubic, leaveTangent=-0.2689, time=1, value=0}
         ]}}}
      Scale Curve=&MakeCustomFloatFromBool(Bool=$User.bIsFrontEndPreview, True Float=0.5, False Float=1)
     )
     B=&Lerp_Float(
      A=&FloatFromCurve(
       FloatCurve=DataInterface() {propertyValues {Curve {keys [
           {interpMode=RCIM_Cubic, tangentMode=RCTM_User, time=0.1, value=0}
           {arriveTangent=7.2184, interpMode=RCIM_Cubic, tangentMode=RCTM_Break, time=0.4, value=1}
          ]}}}
       CurveIndex=$System.InterpolateOverTime.InterpolateOverTime_Float.Moving Average
      )
      B=1
      Alpha=$System.MusicReactivity
     )
    )
    Scale RGB=&MakeVectorFromLinearColorRGB(Color=&SelectLinearColorFromArray(
      Array Sampling Mode=::NewEnumerator1
      Direct Array Index=&MakeIntFromFloat(Float=$User.VariantSwitch)
      Color Selection Array=DataInterface() {propertyValues {ColorData [<rgba 0.0877 0.5573 0.7656 1>, <rgba 0.076 0.75 0.1273 1>]}}
      Random Seed=?
     ))
    ScaleA=true
    ScaleRGB=true
   )
   &DynamicMaterialParameters(
    Index 0 Param 1=&FloatFromCurve(FloatCurve=DataInterface() {propertyValues {Curve {keys [
         {arriveTangent=-4.0585, interpMode=RCIM_Cubic, leaveTangent=-4.0585, tangentMode=RCTM_User, time=0, value=1}
         {arriveTangent=-1.0113, interpMode=RCIM_Cubic, leaveTangent=-1.0113, tangentMode=RCTM_User, time=1, value=-1.5}
        ]}}})
    Module.Index0Param0Write=true
    Module.Index0Param1Write=true
    Module.Index0Param2Write=true
    Module.Index0Param3Write=true
   )
  ]
  rendererClasses [&NiagaraMeshRendererProperties]
  renderers [{rendererClass=&NiagaraMeshRendererProperties, rendererIndex=0}]
 }
 Wave {
  emitterUpdateScript [
   &EmitterState(Inactive Response=::NewEnumerator1, Life Cycle Mode=::NewEnumerator1, Loop Duration=0.5, UseLoopDelay=false)
   &SpawnBurst_Instantaneous(
    Spawn Count=2
    Spawn Probability=&MakeCustomFloatFromBool(
     Bool=$User.bIsFrontEndPreview
     True Float=1
     False Float=&Lerp_Float(
      A=&MakeCustomFloatFromBool(
       Bool=&SetBoolByFloatComparison(A=$NPC.FortniteNPC.EventAudio_Amplitude, B=0, Comparison Type=::NewEnumerator4)
       True Float=0
       False Float=1
      )
      B=1
      Alpha=$System.MusicReactivity
     )
    )
    Use Spawn Probability=true
    UseLoopCountLimit=false
   )
  ]
  particleSpawnScript [
   &PartitionParticles()
   &InitializeParticle(
    Color Mode=::NewEnumerator0
    Lifetime=0.55
    Mesh Scale=<0.76 0.76 0.6>
    Mesh Scale Mode=::NewEnumerator4
    Position Offset=&SelectVectorFromArray(
     Direct Array Index=$Particles.PartitionParticles.Partition
     Vector Selection Array=DataInterface() {propertyValues {FloatData [<7 -6 10>, <-8 -6 -11>]}}
    )
    UsePositionOffset=true
    Write Material Random=true
   )
   &InitialMeshOrientation(Mesh Orientation Mode=::NewEnumerator2, Rotation=<0.25 0 0>, Use Rotation Vector=true)
  ]
  particleUpdateScript [
   &ParticleState()
   &ScaleMeshSize(Scale Factor=&VectorFromFloat(Value=&Multiply_Float(
      A=1.3
      B=&Lerp_Float(
       A=&MakeCustomFloatFromBool(
        Bool=$User.bIsFrontEndPreview
        True Float=&Lerp_Float(
         A=0.6
         B=&Multiply_Float(
          A=1
          B=&FloatFromCurve(
           FloatCurve=DataInterface() {propertyValues {Curve {keys [<at 0 1>,<at 1 0.4>]}}}
           CurveIndex=$NPC.FortniteNPC.EventAudio_Amplitude
           Scale Curve=1.2
          )
         )
         Alpha=$System.InterpolateOverTime.InterpolateOverTime_Float.Moving Average
        )
        False Float=&Power(
         a=&Multiply_Float(A=$System.InterpolateOverTime.InterpolateOverTime_Float.Moving Average, B=$NPC.FortniteNPC.EventAudio_Amplitude)
         b=0.3
        )
       )
       B=0.9
       Alpha=$System.MusicReactivity
      )
     )))
   &DynamicMaterialParameters(
    Index 0 Param 1=1
    Index 0 Param 2=0.08
    Index 0 Param 3=&MakeCustomFloatFromBool(
     Bool=$User.bIsFrontEndPreview
     True Float=&Lerp_Float(A=0.7, B=1.5, Alpha=&ClampFloat(Float=$System.InterpolateOverTime.InterpolateOverTime_Float.Moving Average))
     False Float=&Lerp_Float(
      A=&Lerp_Float(A=0.5, B=1.6, Alpha=&ClampFloat(Float=$System.InterpolateOverTime.InterpolateOverTime_Float.Moving Average))
      B=1
      Alpha=$System.MusicReactivity
     )
    )
    Module.Index0Float0Write=true
    Module.Index0Float1Write=true
    Module.Index0Float2Write=true
    Module.Index0Float3Write=true
   )
   &ScaleColor(
    Scale Alpha=&Multiply_Float(
     A=&SelectFloatFromArray(
      Direct Array Index=&MakeIntFromFloat(Float=$User.VariantSwitch)
      Float Selection Array=DataInterface() {propertyValues {FloatData [0.5,0.3]}}
     )
     B=&Lerp_Float(
      A=&MakeCustomFloatFromBool(
       Bool=$User.bIsFrontEndPreview
       True Float=&Lerp_Float(A=0.5, B=1, Alpha=$System.InterpolateOverTime.InterpolateOverTime_Float.Moving Average)
       False Float=&Multiply_Float(
        A=&FloatFromCurve(
         FloatCurve=DataInterface() {propertyValues {Curve {keys [
             {interpMode=RCIM_Cubic, tangentMode=RCTM_User, time=0.1, value=0}
             {arriveTangent=7.2184, interpMode=RCIM_Cubic, tangentMode=RCTM_Break, time=0.35, value=1}
            ]}}}
         CurveIndex=$System.InterpolateOverTime.InterpolateOverTime_Float.Moving Average
        )
        B=1
       )
      )
      B=1
      Alpha=$System.MusicReactivity
     )
    )
    Scale RGB=&MakeVectorFromLinearColorRGB(Color=&SelectLinearColorFromArray(
      Array Sampling Mode=::NewEnumerator1
      Direct Array Index=&MakeIntFromFloat(Float=$User.VariantSwitch)
      Color Selection Array=DataInterface() {propertyValues {ColorData [<rgba 0 0.276 0.1927 1>, <rgba 0.3099 0.4 0.0264 1>]}}
      Random Seed=?
     ))
    ScaleA=true
    ScaleRGB=true
   )
   /SetVariables(Particles.Color002=&MakeLinearColorFromVectorAndFloat(
     "Vector (RGB)"=&Multiply_VectorByFloat(
      Vector=&MakeCustomVectorFromBool(
       Bool=$User.bIsFrontEndPreview
       True Vector=&MakeVectorFromLinearColorRGB(Color=&SelectLinearColorFromArray(
         Array Sampling Mode=::NewEnumerator1
         Direct Array Index=&MakeIntFromFloat(Float=$User.VariantSwitch)
         Color Selection Array=DataInterface() {propertyValues {ColorData [<rgba 0.1372 3 0.5281 1>, <rgba 0.0469 1 0.2758 1>]}}
         Random Seed=?
        ))
       False Vector=&MakeVectorFromLinearColorRGB(Color=&SelectLinearColorFromArray(
         Array Sampling Mode=::NewEnumerator1
         Direct Array Index=&MakeIntFromFloat(Float=$User.VariantSwitch)
         Color Selection Array=DataInterface() {propertyValues {ColorData [<rgba 3 2.0543 0 1>, <rgba 0.0342 0.7292 0.2011 1>]}}
         Random Seed=?
        ))
      )
      Float=&Lerp_Float(
       A=1
       B=2
       Alpha=&Power(
        a=&Multiply_Float(A=$System.InterpolateOverTime.InterpolateOverTime_Float.Moving Average, B=$NPC.FortniteNPC.EventAudio_Amplitude)
        b=10
       )
      )
     )
     "Float (Alpha)"=&Multiply_Float(
      A=1
      B=&Lerp_Float(
       A=&MakeCustomFloatFromBool(
        Bool=$User.bIsFrontEndPreview
        True Float=&Lerp_Float(A=0.5, B=1, Alpha=$System.InterpolateOverTime.InterpolateOverTime_Float.Moving Average)
        False Float=&Multiply_Float(
         A=&FloatFromCurve(
          FloatCurve=DataInterface() {propertyValues {Curve {keys [
              {interpMode=RCIM_Cubic, tangentMode=RCTM_User, time=0.1, value=0}
              {arriveTangent=7.2184, interpMode=RCIM_Cubic, tangentMode=RCTM_Break, time=0.35, value=1}
             ]}}}
          CurveIndex=$System.InterpolateOverTime.InterpolateOverTime_Float.Moving Average
         )
         B=&SelectFloatFromArray(
          Direct Array Index=&MakeIntFromFloat(Float=$User.VariantSwitch)
          Float Selection Array=DataInterface() {propertyValues {FloatData [1,3]}}
         )
        )
       )
       B=1
       Alpha=$System.MusicReactivity
      )
     )
    )) {isSetParameters=true}
   /SetVariables(Particles.scale_small:Vector3f=&Multiply_VectorByFloat(Vector=$Particles.Scale, Float=0.75)) {isSetParameters=true}
   /SetVariables(Particles.Color_light02:LinearColor=&MakeLinearColorFromVectorAndFloat(
     "Vector (RGB)"=&Multiply_VectorByFloat(
      Vector=&MakeVectorFromLinearColorRGB(Color=&SelectLinearColorFromArray(
        Array Sampling Mode=::NewEnumerator1
        Direct Array Index=&MakeIntFromFloat(Float=$User.VariantSwitch)
        Color Selection Array=DataInterface() {propertyValues {ColorData [<rgba 3 3 3 1>, <rgba 0.8 0.6797 0.04 1>]}}
        Random Seed=?
       ))
      Float=&Multiply_Float(
       A=&Lerp_Float(
        A=1
        B=3
        Alpha=&Power(
         a=&Multiply_Float(A=$System.InterpolateOverTime.InterpolateOverTime_Float.Moving Average, B=$NPC.FortniteNPC.EventAudio_Amplitude)
         b=10
        )
       )
       B=&SelectFloatFromArray(
        Direct Array Index=&MakeIntFromFloat(Float=$User.VariantSwitch)
        Float Selection Array=DataInterface() {propertyValues {FloatData [1,0.7]}}
       )
      )
     )
     "Float (Alpha)"=&Multiply_Float(
      A=1
      B=&Lerp_Float(
       A=&MakeCustomFloatFromBool(
        Bool=$User.bIsFrontEndPreview
        True Float=&Lerp_Float(
         A=0
         B=&Lerp_Float(A=0.5, B=1, Alpha=$System.InterpolateOverTime.InterpolateOverTime_Float.Moving Average)
         Alpha=&SelectFloatFromArray(
          Direct Array Index=&MakeIntFromFloat(Float=$User.VariantSwitch)
          Float Selection Array=DataInterface() {propertyValues {FloatData [0,1]}}
         )
        )
        False Float=&Multiply_Float(
         A=&FloatFromCurve(
          FloatCurve=DataInterface() {propertyValues {Curve {keys [
              {interpMode=RCIM_Cubic, tangentMode=RCTM_User, time=0.1, value=0}
              {arriveTangent=7.2184, interpMode=RCIM_Cubic, tangentMode=RCTM_Break, time=0.35, value=1}
             ]}}}
          CurveIndex=$System.InterpolateOverTime.InterpolateOverTime_Float.Moving Average
         )
         B=&SelectFloatFromArray(
          Direct Array Index=&MakeIntFromFloat(Float=$User.VariantSwitch)
          Float Selection Array=DataInterface() {propertyValues {FloatData [0,3]}}
         )
        )
       )
       B=&SelectFloatFromArray(
        Direct Array Index=&MakeIntFromFloat(Float=$User.VariantSwitch)
        Float Selection Array=DataInterface() {propertyValues {FloatData [0,1]}}
       )
       Alpha=$System.MusicReactivity
      )
     )
    )) {isSetParameters=true}
  ]
  rendererClasses [&NiagaraMeshRendererProperties]
  renderers [
   {rendererClass=&NiagaraMeshRendererProperties, rendererIndex=0}
   {rendererClass=&NiagaraMeshRendererProperties, rendererIndex=1}
   {rendererClass=&NiagaraMeshRendererProperties, rendererIndex=2}
   {rendererClass=&NiagaraMeshRendererProperties, rendererIndex=3}
  ]
 }
 light {
  emitterUpdateScript [
   &EmitterState(Inactive Response=::NewEnumerator1, Life Cycle Mode=::NewEnumerator1, Loop Duration=1, UseLoopDelay=false)
   &SpawnBurst_Instantaneous(
    Spawn Probability=&MakeCustomFloatFromBool(
     Bool=$User.bIsFrontEndPreview
     True Float=1
     False Float=&Lerp_Float(
      A=&MakeCustomFloatFromBool(
       Bool=&SetBoolByFloatComparison(A=$NPC.FortniteNPC.EventAudio_Amplitude, B=0, Comparison Type=::NewEnumerator4)
       True Float=0
       False Float=1
      )
      B=1
      Alpha=$System.MusicReactivity
     )
    )
    Use Spawn Probability=true
    UseLoopCountLimit=false
   )
  ]
  particleSpawnScript [&InitializeParticle(Position Offset=<0 -20 0>, UsePositionOffset=true, Write Material Random=true)]
  particleUpdateScript [
   &ParticleState()
   &FN_Light_Attributes(
    Color=&ColorFromCurve(ColorCurve=DataInterface() {propertyValues {
       AlphaCurve {keys []}
       BlueCurve {keys [<at 0.1541 1>, {arriveTangent=-2.0496, interpMode=RCIM_Cubic, leaveTangent=-2.0496, time=0.3481, value=0}, <at 0.642 0>, <at 0.811 1>]}
       GreenCurve {keys [
         <at 0.1541 0.2948>
         {arriveTangent=1.4453, interpMode=RCIM_Cubic, leaveTangent=1.4453, time=0.3481, value=1}
         <at 0.642 1>
         <at 0.811 0.2948>
        ]}
       RedCurve {keys [
         <at 0.1541 0.0238>
         {arriveTangent=0.031, interpMode=RCIM_Cubic, leaveTangent=0.031, time=0.3481, value=0.0389}
         <at 0.642 0.0389>
         <at 0.811 0.0238>
        ]}
      }})
    Create Color Attribute=true
    Create Exponent Attribute=true
    Create Position Attribute=true
    Create Radius Attribute=true
    Exponent=2
    Position=<0 -20 0>
    Radius=80
    Scale Color=&Multiply_Float(
     A=15
     B=&Lerp_Float(
      A=&MakeCustomFloatFromBool(
       Bool=$User.bIsFrontEndPreview
       True Float=&Lerp_Float(A=0.2, B=1, Alpha=$System.InterpolateOverTime.InterpolateOverTime_Float.Moving Average)
       False Float=&Lerp_Float(A=0, B=1, Alpha=$System.InterpolateOverTime.InterpolateOverTime_Float.Moving Average)
      )
      B=1
      Alpha=$System.MusicReactivity
     )
    )
   )
  ]
  rendererClasses [&NiagaraLightRendererProperties, &NiagaraSpriteRendererProperties]
  renderers [
   {rendererClass=&NiagaraLightRendererProperties, rendererIndex=0}
   {rendererClass=&NiagaraSpriteRendererProperties, rendererIndex=1}
  ]
 }
}
summary {
 systemName=NS_Backpack_LocalTalent_Flare_PocketBurst
 emitters [
  {emitterName=Rays, rendererClasses [&NiagaraMeshRendererProperties]}
  {emitterName=Star_mesh, rendererClasses [&NiagaraMeshRendererProperties]}
  {emitterName=Sprites, rendererClasses [&NiagaraSpriteRendererProperties]}
  {emitterName=Star_Glowmesh, rendererClasses [&NiagaraMeshRendererProperties]}
  {emitterName=Flare, rendererClasses [&NiagaraMeshRendererProperties]}
  {emitterName=Wave, rendererClasses [&NiagaraMeshRendererProperties]}
  {emitterName=MusicLine, rendererClasses [&NiagaraMeshRendererProperties]}
  {emitterName=light, rendererClasses [&NiagaraLightRendererProperties, &NiagaraSpriteRendererProperties]}
 ]
 userVariables [
  User.bIsFrontEnd {type=NiagaraBool}
  User.bIsFrontEndPreview {type=NiagaraBool}
  User.VariantSwitch {type=NiagaraFloat, defaultValue=0, description=""}
 ]
}
