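# Animation Blueprint for the LocalTalent Flare female backpack, adding secondary motion and pose correction.
# Five Anim Dynamics nodes spring the dyn_chain bones, then additive bone offsets and a partial Look At toward the body's left clavicle adjust the pack.
# Spine twist on the body mesh becomes Melee_F, blending extra pack and chain offsets.
# Skydive float and gender values are computed but unused, and several bone nodes are unconnected.
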AnimBlueprint /BRCosmetics/Accessories/FORT_Backpacks/Backpack_F_MED_LocalTalent_Flare/Meshes/F_MED_LocalTalent_Flare_Pack_AnimBP
target_skeleton=&/F_MED_LocalTalent_Flare_Pack_Skeleton
graphs [
 AnimGraph {nodes [
   Look At - Bone: pack!(ComponentPose, Alpha=0.2, LookAtLocation:Vector=Get PelvisLocation() -> (PelvisLocation:Vector)) -> (Pose) {
    alphaBoolBlend {blendOption=Linear}
    boneToModify {boneName=pack}
    lookAt_Axis {axis {y=0, z=-1}}
    lookUp_Axis {axis {x=0, z=1}}
    lookAtTarget=~
   }
   "Transform (Modify) Bone - Bone: pack"(Translation="-2,-2,1", Rotation="-10,0,30") {
    rotationMode=BMM_Additive
    rotationSpace=BCS_BoneSpace
    translationMode=BMM_Additive
    translationSpace=BCS_BoneSpace
    alphaScaleBiasClamp {bInterpResult=true, interpSpeedDecreasing=5, interpSpeedIncreasing=5}
    boneToModify {boneName=pack}
    rotation {pitch=-10, roll=30}
    translation=<-2 -2 1>
   }
   "Transform (Modify) Bone - Bone: pack"(Translation="0,0,0", Rotation="0,0,17") {
    rotationMode=BMM_Additive
    rotationSpace=BCS_BoneSpace
    translationMode=BMM_Additive
    translationSpace=BCS_BoneSpace
    alphaScaleBiasClamp {bInterpResult=true, interpSpeedDecreasing=5}
    boneToModify {boneName=pack}
    rotation {roll=17}
   }
   "Transform (Modify) Bone - Bone: pack_01"(Alpha=1, Translation="0,-90,0", Rotation="0,0,0", Scale="1,1,1") {translationMode=BMM_Additive, translationSpace=BCS_BoneSpace, boneToModify {boneName=pack_01}, translation {y=-90}}
   n1 = Anim Dynamics - Bone: dyn_chain_01(ComponentPose=Local To Component(LocalPose=Input Pose()), Alpha=1, ExternalForce="0,0,0") {
    angularDampingOverride=1
    angularSpringConstant=20000
    bAngularSpring=true
    bOverrideAngularDamping=true
    bUseGravityOverride=true
    gravityScale=0.35
    simulationSpace=RootRelative
    boundBone {boneName=dyn_chain_01}
    componentLinearAccScale=<1 1 1>
    componentLinearVelScale=<1 1 1>
    gravityOverride {z=-980}
    physicsBodyDefinitions [{
      boundBone {boneName=dyn_chain_01}
      boxExtents=<20 30 30>
      constraintSetup {angularLimitsMax=<40 30 5>, angularLimitsMin=<-40 -5 -40>, angularTarget=<1 0 0>}
      localJointOffset=<4 0 0>
     }]
    simSpaceSettings {simSpaceAngularAlpha=0.5}
   }
   n2 = Anim Dynamics - Bone: dyn_chain_02(ComponentPose=@n1, Alpha=1, ExternalForce="0,0,0") {
    angularDampingOverride=1
    angularSpringConstant=20000
    bAngularSpring=true
    bOverrideAngularDamping=true
    bUseGravityOverride=true
    gravityScale=0.35
    simulationSpace=RootRelative
    boundBone {boneName=dyn_chain_02}
    componentLinearAccScale=<1 1 1>
    componentLinearVelScale=<1 1 1>
    gravityOverride {z=-980}
    physicsBodyDefinitions [{
      boundBone {boneName=dyn_chain_02}
      boxExtents=<20 30 30>
      constraintSetup {angularLimitsMax=<40 30 5>, angularLimitsMin=<-40 -5 -40>, angularTarget=<1 0 0>}
      localJointOffset=<4 0 0>
     }]
    simSpaceSettings {simSpaceAngularAlpha=0.5}
   }
   n3 = Anim Dynamics - Bone: dyn_chain_03(ComponentPose=@n2, Alpha=1, ExternalForce="0,0,0") {
    angularDampingOverride=1
    angularSpringConstant=20000
    bAngularSpring=true
    bOverrideAngularDamping=true
    bUseGravityOverride=true
    gravityScale=0.35
    simulationSpace=RootRelative
    boundBone {boneName=dyn_chain_03}
    componentLinearAccScale=<1 1 1>
    componentLinearVelScale=<1 1 1>
    gravityOverride {z=-980}
    physicsBodyDefinitions [{
      boundBone {boneName=dyn_chain_03}
      boxExtents=<20 30 30>
      constraintSetup {angularLimitsMax=<40 30 5>, angularLimitsMin=<-40 -5 -40>, angularTarget=<1 0 0>}
      localJointOffset=<4 0 0>
     }]
    simSpaceSettings {simSpaceAngularAlpha=0.5}
   }
   n4 = Anim Dynamics - Bone: dyn_chain_04(ComponentPose=@n3, Alpha=1, ExternalForce="0,0,0") {
    angularDampingOverride=1
    angularSpringConstant=20000
    bAngularSpring=true
    bOverrideAngularDamping=true
    bUseGravityOverride=true
    gravityScale=0.35
    simulationSpace=RootRelative
    boundBone {boneName=dyn_chain_04}
    componentLinearAccScale=<1 1 1>
    componentLinearVelScale=<1 1 1>
    gravityOverride {z=-980}
    physicsBodyDefinitions [{
      boundBone {boneName=dyn_chain_04}
      boxExtents=<20 30 30>
      constraintSetup {angularLimitsMax=<40 30 5>, angularLimitsMin=<-40 -5 -40>, angularTarget=<1 0 0>}
      localJointOffset=<4 0 0>
     }]
    simSpaceSettings {simSpaceAngularAlpha=0.5}
   }
   n5 = Anim Dynamics - Bone: dyn_chain_05(ComponentPose=@n4, Alpha=1, ExternalForce="0,0,0") {
    angularDampingOverride=1
    angularSpringConstant=20000
    bAngularSpring=true
    bOverrideAngularDamping=true
    bUseGravityOverride=true
    gravityScale=0.35
    simulationSpace=RootRelative
    boundBone {boneName=dyn_chain_05}
    componentLinearAccScale=<1 1 1>
    componentLinearVelScale=<1 1 1>
    gravityOverride {z=-980}
    physicsBodyDefinitions [{
      boundBone {boneName=dyn_chain_05}
      boxExtents=<20 30 30>
      constraintSetup {angularLimitsMax=<40 30 5>, angularLimitsMin=<-40 -5 -40>, angularTarget=<1 0 0>}
      localJointOffset=<4 0 0>
     }]
    simSpaceSettings {simSpaceAngularAlpha=0.5}
   }
   n6 = "Transform (Modify) Bone - Bone: dyn_chain_01"(ComponentPose=@n5, Alpha=1, Translation="0,0,0", Rotation="-40,0,0", Scale="1,1,1") {
    rotationMode=BMM_Additive
    rotationSpace=BCS_BoneSpace
    alphaScaleBiasClamp {bInterpResult=true, interpSpeedDecreasing=5, interpSpeedIncreasing=5}
    boneToModify {boneName=dyn_chain_01}
    rotation {pitch=-40}
   }
   n7 = "Transform (Modify) Bone - Bone: packMover_01"(ComponentPose=@n6, Alpha=1, Translation="0,-1,0", Rotation="0,0,0", Scale="1,1,1") {
    rotationMode=BMM_Additive
    rotationSpace=BCS_BoneSpace
    translationMode=BMM_Additive
    translationSpace=BCS_BoneSpace
    boneToModify {boneName=packMover_01}
    translation {y=-1}
   }
   n8 = "Transform (Modify) Bone - Bone: packMover_02"!(ComponentPose=@n7, bAlphaBoolEnabled:bool=True, Translation="0,-1.5,0", Rotation="0,0,0", Scale="1,1,1") -> (Pose) {
    alphaInputType=Bool
    rotationMode=BMM_Additive
    rotationSpace=BCS_BoneSpace
    translationMode=BMM_Additive
    translationSpace=BCS_BoneSpace
    boneToModify {boneName=packMover_02}
    translation {y=-1.5}
   }
   n9 = "Transform (Modify) Bone - Bone: root"(ComponentPose=@n8, Translation="0,-1,0", Rotation="0,0,17") {
    rotationMode=BMM_Additive
    rotationSpace=BCS_BoneSpace
    translationMode=BMM_Additive
    translationSpace=BCS_BoneSpace
    alphaScaleBiasClamp {bInterpResult=true}
    boneToModify {boneName=root}
    rotation {roll=17}
    translation {y=-1}
   }
   n10 = Look At - Bone: pack(ComponentPose=@n9, Alpha=0.2, LookAtLocation=Get clavicleLocation() -> (clavicleLocation:Vector)) {
    alphaBoolBlend {blendOption=Linear}
    alphaScaleBiasClamp {interpSpeedDecreasing=0.5}
    boneToModify {boneName=pack}
    lookAt_Axis {axis {z=1}}
    lookUp_Axis {axis {x=0, z=1}}
    lookAtTarget=~
   }
   n11 = Get Melee_F() -> (Melee_F:double)
   n12 = "Transform (Modify) Bone - Bone: pack_Mover"(ComponentPose=@n10, Alpha=@n11, Translation="0,-3,0", Rotation="0,0,0", Scale="1,1,1") {
    rotationMode=BMM_Additive
    rotationSpace=BCS_BoneSpace
    translationMode=BMM_Additive
    translationSpace=BCS_BoneSpace
    alphaScaleBiasClamp {bInterpResult=true, interpSpeedDecreasing=8}
    boneToModify {boneName=pack_Mover}
    translation {y=-3}
   }
   n13 = "Transform (Modify) Bone - Bone: dyn_chain_01"(ComponentPose=@n12, Alpha=@n11, Translation="0,0,0", Rotation="-30,-10,0", Scale="1,1,1") {
    rotationMode=BMM_Additive
    rotationSpace=BCS_BoneSpace
    translationMode=BMM_Additive
    translationSpace=BCS_BoneSpace
    alphaScaleBiasClamp {bInterpResult=true, interpSpeedDecreasing=5}
    boneToModify {boneName=dyn_chain_01}
    rotation {pitch=-30, yaw=-10}
   }
   Output Pose(Result=Component To Local(ComponentPose=@n13)) {layerGroup=None}
  ]}
 EventGraph {nodes [
   Event BlueprintUpdateAnimation()
   > n1 = Sequence!(execute) -> (then_0, then_1, then_2:Exec)
   n2 = Get bIsSkydiveFloating()
   n3 = Get bIsSkydiveDiveMode()
   > Set FLoating(FLoating:double="To Float (Boolean)"(InBool=OR Boolean(A=@n2, B=@n3))) -> (Output_Get:double)
   Set Melee_F(
    execute=[@n1.then_1, @n11.then]
    Melee_F:double=GetAlphaFromReferenceTransformDelta(
     SkeletalMeshComponent=Get BodyMesh() -> (BodyMesh:Skeletal Mesh Component Object Reference)
     BoneName=spine_05
     BaseName=spine_01
     DeltaComponentType=QuaternionTwist
     InputRange=(X=0,Y=20)
     TwistAxis="0,0,1"
    )
   ) -> (Output_Get:double)
   n4 = Get bodyMesh() -> (bodyMesh:Skeletal Mesh Component Object Reference)
   Is Valid(exec=@n1.then_2, InputObject=@n4)
   > n5 = Sequence!(execute) -> (then_0, then_1)
   > Set clavicleLocation(clavicleLocation:Vector=GetSocketLocation!(self:Scene Component Object Reference=@n4, InSocketName:Name=clavicle_l) -> (ReturnValue:Vector)) -> (Output_Get:Vector)
   Set PelvisLocation(execute=@n5.then_1, PelvisLocation:Vector=GetSocketLocation(self=@n4, InSocketName=pelvis)) -> (Output_Get:Vector)
   Event OnCharacterCustomizationLoaded()
   > n6 = Set Owning Pawn(Owning Pawn:Pawn Object Reference=TryGetPawnOwner()) -> (Output_Get:Pawn Object Reference)
   > Is Valid(InputObject=@n6)
   > n7 = Sequence()
   > n8 = Cast To FortPlayerPawn(Object=Get Owning Pawn() -> (Owning Pawn:Pawn Object Reference))
   > n9 = Set Owning Player Pawn(Owning Player Pawn:Fort Player Pawn Object Reference=@n8) -> (Output_Get:Fort Player Pawn Object Reference)
   > Set BodyMesh(BodyMesh=GetSkeletalMeshForPartType(self=@n9))
   n10 = FindGender!(execute=@n7.then_1, Object:Object Reference=Get Owning Pawn()) -> (True:Exec, False:Exec, OutGender:EFortCustomGender Enum=Invalid)
   > Set isMale(isMale:bool="Equal (Enum)"(A:EFortCustomGender Enum=@n10, B:EFortCustomGender Enum=Male)) -> (Output_Get:bool)
   n11 = Branch(Condition=Get bIsPlayingMeleeAnim())
   Set Melee_F(execute=@n11.else, Melee_F=0)
   TryGetPawnOwner()
   GetAlphaFromReferenceTransformDelta(BoneName=spine_05, BaseName=spine_01, DeltaComponentType=QuaternionTwist, InputRange=(X=-20,Y=20), TwistAxis="0,0,1")
  ]}
]
variables [
 BodyMesh {type=SkeletalMeshComponent}
 BP Rotate {type=Rotator}
 BP Translate {type=Vector}
 clavicleLocation {type=Vector}
 FLoating {type=number}
 isMale {type=boolean}
 Melee_F {type=number}
 Melee_M {type=number}
 Owning Pawn {type=Pawn, default=null}
 Owning Player Pawn {type=FortPlayerPawn}
 PelvisLocation {type=Vector}
]
